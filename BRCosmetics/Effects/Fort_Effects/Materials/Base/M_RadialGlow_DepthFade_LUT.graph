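# Additive unlit particle material for Niagara and sprite effects that draws a soft radial glow with animated colors.
# Time scaled by LightSpeed scrolls through a New Year disco gradient texture as a color lookup, masked by a radial gradient and fed through emissive adjustments.
# A fade term multiplies particle alpha, the radial mask, time-of-day brightness favoring night, and near-camera plus dynamic-parameter depth fade, which mobile skips.
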Material /BRCosmetics/Effects/Fort_Effects/Materials/Base/M_RadialGlow_DepthFade_LUT
graph {
 properties {
  bUsedWithNiagaraSprites=true
  bUsedWithParticleSprites=true
  blendMode=BLEND_Additive
  pixelDepthOffsetMode=PDOM_Legacy
  shadingModel=MSM_Unlit
  usedShadingModels=Unlit
 }
 nodes [
  n1 = Frac(None=Multiply(A=Time(), B=ScalarParameter() {defaultValue=1, parameterName=LightSpeed}))
  n2 = MaterialFunctionCall!(Radius, CenterPosition) {materialFunction=/Engine/Functions/Engine_MaterialFunctions01/Gradient/RadialGradient}
  MaterialFunctionCall(Emissive=Multiply(A=TextureSample(UVs=@n1) {texture=/Game/Athena/Events/NewYear/DiscoNY_Gradient}.RGB, B=@n2.RadialGradient)) {materialFunction=&MF_FX_EmissiveAdjustments}
  n3 = Constant() {r=0.5}
  n4 = ParticleColor()
  n5 = DepthFade(
   Opacity=MaterialFunctionCall(Distance=ScalarParameter() {defaultValue=100, parameterName=NearCameraFadeDistance}) {materialFunction=&Near_Camera_Fade}.Result
   FadeDistance=DynamicParameter() {defaultValue=<rgba 0 0 0 1>, paramNames [FadeDistance,Param2,Param3,Param4]}.FadeDistance
  )
  n6 = Constant() {r=1}
  Multiply(
   A=MaterialFunctionCall(Color_Morning=@n3, Color_Day=@n3, Color_Evening=@n3, Color_Night=Constant() {r=1}) {materialFunction=&Color_Time_of_Day}.Result
   B=Multiply(
    A=Multiply(A=@n4.A, B=@n2.RadialGradient)
    B=ShadingPathSwitch(Default=@n5, Deferred=@n5, Mobile=@n6) {desc=Migrated from FeatureLevelSwitch}
   )
  )
 ]
}
